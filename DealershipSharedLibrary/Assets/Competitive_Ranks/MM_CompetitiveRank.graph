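# Masked unlit master material for competitive rank effects, flagged for mesh particles and Niagara sprites, ribbons and meshes.
# A Fresnel, with Particle Color RGB driving its exponent and a parameter setting base reflectance, feeds emissive through MF_FX_EmissiveAdjustments.
# Particle alpha times the blue channel of a dust-distortion normal map forms the opacity mask.
# The assembled material attributes then pass into the M_FakeWindowReflections function.
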Material /DealershipSharedLibrary/Assets/Competitive_Ranks/MM_CompetitiveRank
graph {
 properties {
  bAutomaticallySetUsageInEditor=false
  bUseMaterialAttributes=true
  bUsedWithMeshParticles=true
  bUsedWithNiagaraMeshParticles=true
  bUsedWithNiagaraRibbons=true
  bUsedWithNiagaraSprites=true
  blendMode=BLEND_Masked
  previewMesh=&/SM_Bronze_Logo
  shadingModel=MSM_Unlit
  usedShadingModels=Unlit
 }
 nodes [
  n1 = ParticleColor()
  n2 = Fresnel(ExponentIn=@n1.RGB, BaseReflectFractionIn=ScalarParameter() {defaultValue=0.2, parameterName=BaseReflectFractionIn}) {baseReflectFraction=0.1}
  n3 = TextureSample() {samplerType=SAMPLERTYPE_Normal, texture=/Game/Effects/Fort_Effects/Textures/T_DustDistortion}
  MaterialFunctionCall!(In=MakeMaterialAttributes!(
    BaseColor
    Metallic
    Specular
    Roughness
    Anisotropy
    EmissiveColor=MaterialFunctionCall(Emissive=@n2) {materialFunction=&MF_FX_EmissiveAdjustments}.Result
    Opacity
    OpacityMask=Multiply(A=@n1.A, B=@n3.B)
    Normal
    Tangent
    WorldPositionOffset
    SubsurfaceColor
    ClearCoat
    ClearCoatRoughness
    AmbientOcclusion
    Refraction
    CustomizedUVs_0
    CustomizedUVs_1
    CustomizedUVs_2
    CustomizedUVs_3
    CustomizedUVs_4
    CustomizedUVs_5
    CustomizedUVs_6
    CustomizedUVs_7
    PixelDepthOffset
    ShadingModel
    Displacement
   )) {materialFunction=/Game/Athena/WinterArt/Materials/M_FakeWindowReflections}
 ]
}
